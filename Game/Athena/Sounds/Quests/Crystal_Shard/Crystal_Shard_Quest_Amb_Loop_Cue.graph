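# Looping ambient sound cue for Crystal Shard quest objects in Athena, classed as in-game mission sound.
# A single wave player feeds the Crystal_Shard_Loop wave straight to the output at boosted volume.
# The cue overrides its attenuation with a short falloff and listener focus that quiets it when off-focus, and overrides concurrency to cap simultaneous instances.
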SoundCue /Game/Athena/Sounds/Quests/Crystal_Shard/Crystal_Shard_Quest_Amb_Loop_Cue
root=WavePlayer_1
nodes [{id=WavePlayer_1, properties {soundWaveAssetPtr=&/Crystal_Shard_Loop}}]
properties {
 attenuationSettings=/Game/Sounds/Attenuation/Fort_Mission_CrystalShard_Att
 bOverrideAttenuation=true
 bOverrideConcurrency=true
 maxDistance=1500
 soundClassObject=/Game/Sounds/FortSoundClassesAndModes/Fort_Mission_IngameSounds
 volumeMultiplier=1.75
 attenuationOverrides {bEnableListenerFocus=true, falloffDistance=1500, nonFocusVolumeAttenuation=0.35, attenuationShapeExtents {x=0}}
 concurrencyOverrides {maxCount=4}
}
