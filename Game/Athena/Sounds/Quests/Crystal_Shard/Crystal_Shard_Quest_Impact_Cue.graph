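# Sound cue for a Crystal Shard quest impact, layering two sets of crystal hit sounds.
# A mixer blends a random mission crystal impact at full level with a random Storm King crystal weak-point impact at half level.
# Each layer passes through a matching fade-out volume envelope and a slight downward pitch randomizer.
# The cue uses Crystal Shard mission attenuation and overrides concurrency to prevent new plays at the limit.
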SoundCue /Game/Athena/Sounds/Quests/Crystal_Shard/Crystal_Shard_Quest_Impact_Cue
root=Mixer_0
nodes [
 {
  id=Enveloper_0
  children [Random_2]
  properties {
   pitchCurve {editorCurveData {keys [{time=0}]}}
   volumeCurve {editorCurveData {keys [<at 0 1>,<at 0.6026 1>,<at 0.7815 0.4363>,<at 1.1158 0>]}}
  }
 }
 {
  id=Enveloper_1
  children [Random_0]
  properties {
   pitchCurve {editorCurveData {keys [{time=0}]}}
   volumeCurve {editorCurveData {keys [<at 0 1>,<at 0.6026 1>,<at 0.7815 0.4363>,<at 1.1158 0>]}}
  }
 }
 {id=Mixer_0, children [Modulator_2,Modulator_0], properties {inputVolume [1,0.5]}}
 {id=Modulator_0, children [Enveloper_1], properties {pitchMax=1, pitchMin=0.9}}
 {id=Modulator_2, children [Enveloper_0], properties {pitchMax=1, pitchMin=0.9}}
 {id=Random_0, children [WavePlayer_0,WavePlayer_1,WavePlayer_2,WavePlayer_3], properties {weights [1,1,1,1]}}
 {id=Random_2, children [WavePlayer_4,WavePlayer_5,WavePlayer_6,WavePlayer_7,WavePlayer_8], properties {weights [1,1,1,1,1]}}
 {id=WavePlayer_0, properties {soundWaveAssetPtr=/Game/Sounds/Fort_Monster_Storm_King/SFX/Final_SFX/StormKing_CrystalWeakPoint_Impact_01}}
 {id=WavePlayer_1, properties {soundWaveAssetPtr=/Game/Sounds/Fort_Monster_Storm_King/SFX/Final_SFX/StormKing_CrystalWeakPoint_Impact_02}}
 {id=WavePlayer_2, properties {soundWaveAssetPtr=/Game/Sounds/Fort_Monster_Storm_King/SFX/Final_SFX/StormKing_CrystalWeakPoint_Impact_03}}
 {id=WavePlayer_3, properties {soundWaveAssetPtr=/Game/Sounds/Fort_Monster_Storm_King/SFX/Final_SFX/StormKing_CrystalWeakPoint_Impact_04}}
 {id=WavePlayer_4, properties {soundWaveAssetPtr=/Game/Sounds/Fort_GamePlay_Sounds/Missions/crystal_impact_01}}
 {id=WavePlayer_5, properties {soundWaveAssetPtr=/Game/Sounds/Fort_GamePlay_Sounds/Missions/crystal_impact_02}}
 {id=WavePlayer_6, properties {soundWaveAssetPtr=/Game/Sounds/Fort_GamePlay_Sounds/Missions/crystal_impact_03}}
 {id=WavePlayer_7, properties {soundWaveAssetPtr=/Game/Sounds/Fort_GamePlay_Sounds/Missions/crystal_impact_04}}
 {id=WavePlayer_8, properties {soundWaveAssetPtr=/Game/Sounds/Fort_GamePlay_Sounds/Missions/crystal_impact_05}}
]
properties {
 attenuationSettings=/Game/Sounds/Attenuation/Fort_Mission_CrystalShard_Att
 bOverrideConcurrency=true
 maxDistance=2800
 pitchMultiplier=0.9
 soundClassObject=/Game/Sounds/FortSoundClassesAndModes/Fort_Mission_IngameSounds
 volumeMultiplier=0.5
 attenuationOverrides {falloffDistance=2600}
 concurrencyOverrides {resolutionRule=PreventNew}
}
