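# Looping ambient sound cue for the Crystal Shard quest, apparently for its pre-activation state, in the quests sound class.
# A mixer layers a quiet pitched-down shard loop, a random pick of three pitched-up Zero Point crystal ambience loops, and a lowered pulse loop.
# It overrides attenuation with listener focus and a short falloff, and caps concurrency at a few simultaneous instances.
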SoundCue /Game/Athena/Sounds/Quests/Crystal_Shard/Crystal_Shard_Quest_Pre_Amb_Loop_Cue
root=Mixer_0
nodes [
 {id=Mixer_0, children [Modulator_1,Modulator_0,Modulator_2], properties {inputVolume [0.05,1,0.65]}}
 {id=Modulator_0, children [Random_0], properties {pitchMax=1.25, pitchMin=1.1}}
 {id=Modulator_1, children [WavePlayer_1], properties {pitchMax=0.9, pitchMin=0.9}}
 {id=Modulator_2, children [WavePlayer_4], properties {pitchMax=0.75, pitchMin=0.75}}
 {id=Random_0, children [WavePlayer_0,WavePlayer_2,WavePlayer_3], properties {weights [1,1,1]}}
 {id=WavePlayer_0, properties {bLooping=true, soundWaveAssetPtr=/Game/Sounds/Apollo/Biomes/ZeroPoint/ZP_Crystal_Amb_Loop_01}}
 {id=WavePlayer_1, properties {bLooping=true, soundWaveAssetPtr=&/Crystal_Shard_Loop}}
 {id=WavePlayer_2, properties {bLooping=true, soundWaveAssetPtr=/Game/Sounds/Apollo/Biomes/ZeroPoint/ZP_Crystal_Amb_Loop_02}}
 {id=WavePlayer_3, properties {bLooping=true, soundWaveAssetPtr=/Game/Sounds/Apollo/Biomes/ZeroPoint/ZP_Crystal_Amb_Loop_03}}
 {id=WavePlayer_4, properties {bLooping=true, soundWaveAssetPtr=&/Crystal_Shard_Pulse_Loop}}
]
properties {
 attenuationSettings=/Game/Athena/Sounds/Events/ZP2/ZP_Distant_Pulse_Att
 bOverrideAttenuation=true
 bOverrideConcurrency=true
 maxDistance=1500
 pitchMultiplier=1.2
 soundClassObject=&Fort_Quests
 volumeMultiplier=1.68
 attenuationOverrides {bEnableListenerFocus=true, falloffDistance=1500, nonFocusVolumeAttenuation=0.35, attenuationShapeExtents {x=0}}
 concurrencyOverrides {maxCount=4}
}
